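# Parent material for Nox's hair strands, rendering skinned hair as camera-facing ribbons.
# World Position Offset widens each ribbon along a view-perpendicular axis, with width from a curve atlas that also supplies a rounded normal blended into the vertex normal.
# Base color combines a UV gradient, pre-skinned-position sphere-mask darkening and Fresnel edge tint; metalness, specular and roughness lerp between parameter ranges.
# The attributes then pass through CharacterSkinFX and MF_TechArtOverrides.
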Material /BRCosmetics/Characters/Player/Female/Medium/Bodies/F_MED_Nox/Materials/M_Nox_HairStrand_Parent
graph {
 properties {
  bTangentSpaceNormal=false
  bUseMaterialAttributes=true
  bUsedWithMeshDeformer=true
  bUsedWithSkeletalMesh=true
  blendMode=BLEND_Masked
  pixelDepthOffsetMode=PDOM_Legacy
  refractionMethod=RM_IndexOfRefraction
  shadingModel=MSM_FromMaterialExpression
  twoSided=true
  usedShadingModels=Default Lit | Subsurface | Clear Coat | Subsurface Profile | Cloth | Eye
 }
 nodes [
  Constant() {r=200}
  n1 = CameraPositionWS()
  n2 = Constant3Vector() {constant=<rgba 0 0 1 0>}
  n3 = MaterialFunctionCall() {materialFunction=&LocalPosition}
  n4 = Normalize(VectorInput=Subtract(A=CrossProduct(A=@n1, B=@n2), B=@n3.Local Position))
  n5 = ObjectOrientation()
  CrossProduct(A=@n4, B=@n5)
  DeriveNormalZ()
  n6 = MaterialFunctionCall(Float2=NamedRerouteUsage()) {materialFunction=&BreakOutFloat2Components}
  n7 = CurveAtlasRowParameter(CurveTime=@n6.G) {
   atlas=/Game/Characters/Shared/CurveAtlas/CharacterHairRibbonAtlas
   curve=/Game/Characters/Player/Female/Medium/Bodies/F_MED_Nox/Curves/C_Nox_Hair
   group=HairRibbon
   parameterName=HairCurve
  }
  n8 = ScalarParameter() {defaultValue=1, group=HairRibbon, parameterName=HairRibbon_Thickness}
  n9 = ComponentMask(None=NamedRerouteUsage()) {r=true}
  n10 = Add(A=Constant() {r=-0.5}, B=@n9)
  n11 = Normalize(VectorInput=Transform(None=Constant3Vector() {constant=<rgba 0 1 0 0>}))
  n12 = CrossProduct(A=Subtract(A=CameraPositionWS(), B=WorldPosition().XYZ), B=@n11)
  n13 = Normalize(VectorInput=@n12)
  n14 = VertexNormalWS()
  n15 = VertexColor()
  n16 = MaterialFunctionCall(Float2=NamedRerouteUsage()) {materialFunction=&BreakOutFloat2Components}
  n17 = ConstantBiasScale(None=Frac(None=@n9)) {bias=-0.5, scale=2}
  n18 = Constant()
  n19 = CurveAtlasRowParameter(CurveTime=@n17) {
   atlas=&CharacterHairRibbonAtlas
   curve=&C_Nox_Hair
   group=HairRibbon
   parameterName=HairCurve
  }
  n20 = MaterialFunctionCall!(Vector1=@n12, Vector2=Multiply(A=@n11) {constB=-1}) {materialFunction=/Engine/Functions/Engine_MaterialFunctions02/Math/CreateThirdOrthogonalVector}
  n21 = MaterialFunctionCall(
   VectorToTransform=AppendVector(A=AppendVector(A=@n17, B=@n18), B=Multiply(A=@n19.G))
   BasisX=Normalize(VectorInput=@n20.Vector1)
   BasisY=Normalize(VectorInput=@n20.Vector2)
   BasisZ=Normalize(VectorInput=@n20.Vector3)
  ) {materialFunction=&Transform3x3Matrix}
  n22 = LinearInterpolate(
   A=@n14
   B=@n21.Transformed Vector
   Alpha=SmoothStep(
    Min=ScalarParameter() {group=HairRibbon, parameterName=NormalBlend_Start}
    Max=ScalarParameter() {defaultValue=1, group=HairRibbon, parameterName=Normal_Blend_End}
    Value=StaticSwitchParameter(True=@n15.R, False=@n16.G) {parameterName=UseVexterColorForNormalBlending}
   ) {constMax=0.2}
  )
  n23 = MaterialFunctionCall(Float2=NamedRerouteUsage()) {materialFunction=&BreakOutFloat2Components}
  n24 = Power(Base=@n23.G, Exp=ScalarParameter() {defaultValue=1, group=Nox, parameterName=HairRibbon_basecolor_gradientPower})
  n25 = PreSkinnedPosition()
  n26 = ShaderStageSwitch(PixelShader=VertexInterpolator(VS=@n25).PS, VertexShader=@n25)
  n27 = PreSkinnedLocalBounds()
  n28 = MaterialFunctionCall(Float3=Divide(A=Multiply(A=@n26), B=Subtract(A=@n27.Max, B=@n27.Max)) {constB=50}) {materialFunction=&BreakOutFloat3Components}
  n29 = Constant() {r=-0.35}
  n30 = SphereMask(A=@n28.B, B=Constant() {r=4.6}) {attenuationRadius=0.2, hardnessPercent=0}
  n31 = Saturate(None=Add(A=SphereMask(A=@n28.G, B=@n29) {attenuationRadius=0.5, hardnessPercent=5}, B=@n30))
  n32 = ScalarParameter() {defaultValue=0.5, group=Nox, parameterName=Nox_TipDarkeningAmount}
  n33 = Saturate(None=Add(A=OneMinus(None=@n31), B=OneMinus(None=@n32)) {constB=0.1})
  n34 = LinearInterpolate(
   A=VectorParameter() {group=HairRibbon, parameterName=HairRibbon_BaseColor_b, defaultValue=<rgba 0 0 0 1>}.RGB
   B=VectorParameter() {group=HairRibbon, parameterName=HairRibbon_BaseColor_a, defaultValue=<rgba 1 1 1 1>}.RGB
   Alpha=Saturate(None=Multiply(A=@n24, B=@n33))
  )
  n35 = SmoothStep(Value=@n23.G) {constMax=0.15}
  n36 = Fresnel(
   ExponentIn=ScalarParameter() {defaultValue=2, group=HairRibbon, parameterName=HairRibbon_FresnelExp}
   Normal=LinearInterpolate(A=@n22, B=@n14, Alpha=NamedRerouteUsage())
  ) {baseReflectFraction=0.01, exponent=2}
  n37 = ScalarParameter() {defaultValue=0.1, group=HairRibbon, parameterName=HairRibbon_FresnelAmount}
  n38 = NamedRerouteUsage()
  n39 = SetMaterialAttributes!(
   MaterialAttributes
   World Position Offset=Multiply(A=Multiply(A=@n7.R, B=@n8), B=Multiply(A=@n10, B=@n13))
   Normal=@n22
   Base Color=LinearInterpolate(
    A=LinearInterpolate(B=@n34, Alpha=@n35)
    B=VectorParameter() {group=HairRibbon, parameterName=HairRibbon_EdgeTint, defaultValue=<rgba 0 0 0 1>}.RGB
    Alpha=Multiply(A=@n36, B=@n37) {constB=0.1}
   )
   Curvature or Metallic=LinearInterpolate(
    A=ScalarParameter() {defaultValue=1, group=HairRibbon, parameterName=HairRibbon_Metalness_Max}
    B=ScalarParameter() {group=HairRibbon, parameterName=HairRibbon_Metalness_Min}
    Alpha=@n38
   ) {constB=0}
   Specular=LinearInterpolate(
    A=ScalarParameter() {defaultValue=0.1, group=HairRibbon, parameterName=HairRibbon_Specular_Min}
    B=ScalarParameter() {defaultValue=0.8, group=HairRibbon, parameterName=HairRibbon_Specular_Max}
    Alpha=@n38
   ) {constB=0.75}
   Roughness=LinearInterpolate(
    A=ScalarParameter() {defaultValue=0.8, group=HairRibbon, parameterName=HairRibbon_Roughness_Max}
    B=ScalarParameter() {defaultValue=0.5, group=HairRibbon, parameterName=HairRibbon_Roughness_Min}
    Alpha=@n38
   ) {constB=0.75}
   Opacity Mask=Constant() {r=1}
  ) {attributeSetTypes [
    #1
    #2
    #3
    #4
    #5
    #6
    #7
   ]}
  MaterialFunctionCall(In=MaterialFunctionCall(MaterialAttributes=@n39) {materialFunction=&CharacterSkinFX}.Result) {materialFunction=&MF_TechArtOverrides}
  MaterialFunctionCall!(XY vector, Z sign, Default Normalization vector) {materialFunction=/Engine/Functions/Engine_MaterialFunctions02/Math/DeriveNormalZ_Function}
  MaterialFunctionCall(
   X=ConstantBiasScale(None=DotProduct(A=@n4, B=@n21.Transformed Vector))
   Y=MaterialFunctionCall(Float2=NamedRerouteUsage()) {materialFunction=&BreakOutFloat2Components}.G
  ) {materialFunction=&MakeFloat2}
  Max()
  Multiply(A=@n30) {constB=1.2}
  Multiply(A=@n7.R, B=@n6.G)
  Multiply(A=@n3.Local Position, B=@n5)
  Multiply(A=@n28.B)
  n40 = Add(A=MaterialFunctionCall(Float2=TextureCoordinate()) {materialFunction=&BreakOutFloat2Components}.R) {constB=-0.5}
  n41 = OneMinus(None=Multiply(A=Abs(None=@n40)) {constB=2})
  n42 = Multiply(A=TextureCoordinate() {vTiling=5}, B=ScalarParameter() {defaultValue=1, group=HairRibbon, parameterName=HairRibbon_NoiseTiling})
  NamedRerouteDeclaration(None=MaterialFunctionCall(
    Transition Phase=Power(Base=@n41) {constExponent=0.8}
    Height Texture=TextureSample(UVs=@n42) {texture=/Game/Global/Textures/Default/Tileable_Patterns/T_Blowing_Smoke_Pattern2_Tile_DISP}.R
    Contrast=Constant()
   ) {materialFunction=&HeightLerp}.Alpha) {name=RibbonNoise, nodeColor=<rgba 0 0.2031 1 1>}
  n43 = TextureCoordinate() {coordinateIndex=2}
  n44 = TextureCoordinate() {coordinateIndex=1}
  n45 = ScalarParameter() {group=HairRibbon, parameterName=HairRibbon_Channel}
  n46 = Abs(None=Sign(None=Subtract(B=@n45) {constA=2}))
  n47 = Sign(None=Subtract(B=ScalarParameter() {group=HairRibbon, parameterName=HairRibbon_Channel}) {constA=0})
  NamedRerouteDeclaration(None=LinearInterpolate(A=TextureCoordinate(), B=LinearInterpolate(A=@n43, B=@n44, Alpha=@n46), Alpha=Abs(None=@n47))) {name=texcord, nodeColor=<rgba 1 0 0.7734 1>}
  SmoothStep(Value=@n28.B) {constMin=0.4}
 ]
}
