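# Physics asset for the dangling wires on the F_MED_Nox female player body cosmetic.
# Capsules on the dyn_wire bones form a two-link upper chain and two single-link lower segments, each starting from a small sphere anchor body.
# Tight swing and twist limits keep the wires fairly stiff.
# Pelvis, spine and right-thigh capsules have no constraints, apparently serving as body collision for the wires.
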PhysicsAsset /BRCosmetics/Characters/Player/Female/Medium/Bodies/F_MED_Nox/Meshes/F_MED_Nox_Wire_Physics
bodies {
 dyn_wire_01 [{shapeType=Capsule, length=11.2074, radius=2.2942, shapeName=dyn_wire_01_capsule, center=<6.3374 0 0>, rotation=<rot 90 0 0>}]
 dyn_wire_02 [{
   shapeType=Capsule
   length=18.5409
   radius=2.0602
   shapeName=dyn_wire_02_capsule
   center=<10.1582 -0.9761 1.0042>
   rotation=<rot 75.7439 178.8216 -169.2251>
  }]
 dyn_wire_lower_03 [{
   shapeType=Capsule
   length=11.3243
   radius=2.0926
   shapeName=dyn_wire_lower_03_capsule
   center=<4 -0.6 -1.8>
   rotation=<rot 86.4075 -11.9998 0.0002>
  }]
 dyn_wire_lower_04_base [{shapeType=Sphere, radius=0.5, shapeName=dyn_wire_lower_04_base_sphere}]
 dyn_wire_lower_04_parent [{
   shapeType=Capsule
   length=6.7827
   radius=1.7268
   shapeName=dyn_wire_lower_04_parent_capsule
   center=<3.7414 0 0>
   rotation=<rot 90 52.6125 69.0243>
  }]
 dyn_wire_lower_parent [{shapeType=Sphere, radius=0.6063, shapeName=dyn_wire_lower_parent_sphere}]
 dyn_wire_parent [{shapeType=Sphere, radius=0.5, shapeName=dyn_wire_parent_sphere}]
 pelvis [{shapeType=Capsule, length=13.3302, radius=10.9198, shapeName=pelvis_capsule, center=<-3.9815 0.6747 0>}]
 spine_02 [{
   shapeType=Capsule
   length=14.2591
   radius=10.4127
   shapeName=spine_02_capsule
   center=<-1.8656 0.5006 0>
   rotation=<rot 90 179.7762 179.4237>
  }]
 spine_05 [{shapeType=Capsule, length=8.0461, radius=13.6042, shapeName=spine_05_capsule, center=<-10.172 -0.7688 0>}]
 thigh_r [{shapeType=Capsule, length=21.8125, radius=8.1875, shapeName=thigh_r_capsule, center=<22.4773 0 0>, rotation=<rot 90 89.4327 83.9136>}]
}
constraints [
 {bone1Name=dyn_wire_01, bone2Name=dyn_wire_parent, swing1LimitDegrees=5, swing2LimitDegrees=5, twistLimitDegrees=3}
 {bone1Name=dyn_wire_02, bone2Name=dyn_wire_01, swing1LimitDegrees=8, swing2LimitDegrees=8, twistLimitDegrees=3}
 {bone1Name=dyn_wire_lower_03, bone2Name=dyn_wire_lower_parent, swing1LimitDegrees=6, swing2LimitDegrees=6, twistLimitDegrees=3}
 {bone1Name=dyn_wire_lower_04_parent, bone2Name=dyn_wire_lower_04_base, swing1LimitDegrees=6, swing2LimitDegrees=6, twistLimitDegrees=5}
]
