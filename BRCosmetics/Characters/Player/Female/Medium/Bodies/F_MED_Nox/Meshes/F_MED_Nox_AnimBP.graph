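# Post-process style AnimBlueprint for the F_MED_Nox player body, copying the parent mesh pose and adding accessory motion and material cloth effects.
# Right-knee bend drives wire bone offsets before RigidBody simulation; constraints place remote, pouch and skates.
# Skate bones spin on real time while the pawn moves.
# Updates feed hand motion to WPO cloth materials; skydiving shifts the backpack attach bone and enables cloth wind.
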AnimBlueprint /BRCosmetics/Characters/Player/Female/Medium/Bodies/F_MED_Nox/Meshes/F_MED_Nox_AnimBP
target_skeleton=&/F_MED_Nox_Skeleton
graphs [
 AnimGraph {nodes [
   n1 = "Transform (Modify) Bone - Bone: attach_backpack"!(
    ComponentPose=Local To Component(LocalPose=Copy Pose From Mesh() {bUseAttachedParent=true})
    bAlphaBoolEnabled:bool=Get bIsSkyDiving()
    Translation="0,-8,-2"
    Rotation="0,0,0"
    Scale="1,1,1"
   ) -> (Pose) {
    alphaInputType=Bool
    rotationMode=BMM_Replace
    rotationSpace=BCS_BoneSpace
    translationMode=BMM_Replace
    translationSpace=BCS_BoneSpace
    alphaBoolBlend {blendOption=Linear}
    boneToModify {boneName=attach_backpack}
    translation {y=-8, z=-2}
   }
   n2 = Get LegRotation() -> (LegRotation:double)
   n3 = "Transform (Modify) Bone - Bone: dyn_wire_lower_parent"(ComponentPose=@n1, Alpha=@n2, Translation="-4,-4,0", Rotation="0,0,0", Scale="1,1,1") {
    translationMode=BMM_Additive
    translationSpace=BCS_BoneSpace
    alphaBoolBlend {blendOption=Linear}
    boneToModify {boneName=dyn_wire_lower_parent}
    translation=<-4 -4>
   }
   n4 = "Transform (Modify) Bone - Bone: dyn_wire_lower_parent"(ComponentPose=@n3, Alpha=@n2, Translation="-5,-7,-6", Rotation="0,0,0", Scale="1,1,1") {
    translationMode=BMM_Additive
    translationSpace=BCS_BoneSpace
    alphaBoolBlend {blendOption=Linear}
    boneToModify {boneName=dyn_wire_lower_parent}
    translation=<-5 -7 -6>
   }
   n5 = "Transform (Modify) Bone - Bone: dyn_wire_parent"(ComponentPose=@n4, Alpha=@n2, Translation="0,3,3", Rotation="0,0,0", Scale="1,1,1") {
    translationMode=BMM_Additive
    translationSpace=BCS_BoneSpace
    alphaBoolBlend {blendOption=Linear}
    boneToModify {boneName=dyn_wire_parent}
    translation {y=3, z=3}
   }
   n6 = "Transform (Modify) Bone - Bone: dyn_wire_lower_parent"(ComponentPose=@n5, Alpha=1, Translation="0,0,-2", Rotation="0,0,0", Scale="1,1,1") {
    translationMode=BMM_Additive
    translationSpace=BCS_BoneSpace
    alphaBoolBlend {blendOption=Linear}
    boneToModify {boneName=dyn_wire_lower_parent}
    translation {z=-2}
   }
   n7 = "Transform (Modify) Bone - Bone: dyn_wire_lower_parent"(ComponentPose=@n6, Alpha=@n2, Translation="0,-2.235,-5", Rotation="0,0,0", Scale="1,1,1") {
    translationMode=BMM_Additive
    translationSpace=BCS_BoneSpace
    alphaBoolBlend {blendOption=Linear}
    boneToModify {boneName=dyn_wire_lower_parent}
    translation {y=-2.235, z=-5}
   }
   n8 = "Transform (Modify) Bone - Bone: dyn_wire_lower_parent"(ComponentPose=@n7, Alpha=@n2, Translation="-4,-1,-4", Rotation="0,0,0", Scale="0.25,1,1") {
    rotationMode=BMM_Additive
    rotationSpace=BCS_BoneSpace
    scaleMode=BMM_Replace
    scaleSpace=BCS_BoneSpace
    translationMode=BMM_Additive
    translationSpace=BCS_BoneSpace
    alphaBoolBlend {blendOption=Linear}
    boneToModify {boneName=dyn_wire_lower_parent}
    scale {x=0.25}
    translation=<-4 -1 -4>
   }
   n9 = "Transform (Modify) Bone - Bone: dyn_wire_lower_04_base"(ComponentPose=@n8, Alpha=@n2, Translation="-1,-4,0", Rotation="0,0,0", Scale="0.5,1,1") {
    rotationMode=BMM_Additive
    rotationSpace=BCS_BoneSpace
    scaleMode=BMM_Replace
    scaleSpace=BCS_BoneSpace
    translationMode=BMM_Additive
    translationSpace=BCS_BoneSpace
    alphaBoolBlend {blendOption=Linear}
    boneToModify {boneName=dyn_wire_lower_04_base}
    scale {x=0.5}
    translation=<-1 -4>
   }
   n10 = RigidBody(ComponentPose=@n9) {
    bClampLinearTranslationLimitToRefPose=true
    bOverrideWorldGravity=true
    overridePhysicsAsset=&/F_MED_Nox_Wire_Physics
    simulationSpace=BaseBoneSpace
    alphaBoolBlend {blendOption=Linear}
    baseBoneRef {boneName=spine_05}
    componentAppliedLinearAccClamp=<1100 1100 1000>
    componentLinearAccScale=<0.5 0.5 0.5>
    componentLinearVelScale=<0.5 0.5 0.5>
    overrideWorldGravity {z=-980}
    simSpaceSettings {
     maxAngularAcceleration=500
     maxAngularVelocity=3
     maxLinearAcceleration=400
     maxLinearVelocity=250
     velocityScaleZ=0
     worldAlpha=1
     externalLinearDragV=<0.1 0.1>
    }
   }
   n11 = "Constraint - const_remote_base (PPPP)"!(
    ComponentPose=@n10
    Alpha=1
    ConstraintWeights_0=0.148
    ConstraintWeights_1:float=0.1593
    ConstraintWeights_2:float=0.4561
    ConstraintWeights_3:float=0.2366
   ) -> (Pose) {
    alphaBoolBlend {blendOption=Linear}
    boneToModify {boneName=const_remote_base}
    constraintSetup [
     {targetBone {boneName=spine_04}}
     {targetBone {boneName=spine_05}}
     {targetBone {boneName=clavicle_r}}
     {targetBone {boneName=upperarm_twist_01_r}}
    ]
    constraintWeights [1,1,1,1]
   }
   n12 = "Constraint - const_pouch_base (PPP)"!(ComponentPose=@n11, Alpha=1, ConstraintWeights_0=0.1083, ConstraintWeights_1:float=0.5652, ConstraintWeights_2:float=0.3265) -> (Pose) {
    alphaBoolBlend {blendOption=Linear}
    boneToModify {boneName=const_pouch_base}
    constraintSetup [{targetBone {boneName=pelvis}}, {targetBone {boneName=thigh_twist_01_r}}, {targetBone {boneName=thigh_r}}]
    constraintWeights [1,1,1]
   }
   n13 = "Constraint - dyn_skate_a_r (P)"!(ComponentPose=@n12, Alpha=1, ConstraintWeights_0=1, ConstraintWeights_1:float=0) -> (Pose) {
    alphaBoolBlend {blendOption=Linear}
    boneToModify {boneName=dyn_skate_a_r}
    constraintSetup [{targetBone {boneName=ball_r}}]
    constraintWeights [1,0]
   }
   n14 = "Constraint - dyn_skate_a_l (P)"!(ComponentPose=@n13, Alpha=1, ConstraintWeights_0=1, ConstraintWeights_1:float=0) -> (Pose) {
    alphaBoolBlend {blendOption=Linear}
    boneToModify {boneName=dyn_skate_a_l}
    constraintSetup [{targetBone {boneName=ball_l}}]
    constraintWeights [1,0]
   }
   n15 = Get Spinning_Rotator() -> (Spinning_Rotator:Rotator)
   n16 = Get PlayerMovement() -> (PlayerMovement:double)
   n17 = "Transform (Modify) Bone - Bone: dyn_skate_a_l"(ComponentPose=@n14, Alpha=@n16, Translation="0,0,0", Rotation=@n15, Scale="1,1,1") {rotationMode=BMM_Additive, rotationSpace=BCS_BoneSpace, alphaBoolBlend {blendOption=Linear}, boneToModify {boneName=dyn_skate_a_l}}
   n18 = "Transform (Modify) Bone - Bone: dyn_skate_b_l"(ComponentPose=@n17, Alpha=@n16, Translation="0,0,0", Rotation=@n15, Scale="1,1,1") {rotationMode=BMM_Additive, rotationSpace=BCS_BoneSpace, alphaBoolBlend {blendOption=Linear}, boneToModify {boneName=dyn_skate_b_l}}
   n19 = "Transform (Modify) Bone - Bone: dyn_skate_c_l"(ComponentPose=@n18, Alpha=@n16, Translation="0,0,0", Rotation=@n15, Scale="1,1,1") {rotationMode=BMM_Additive, rotationSpace=BCS_BoneSpace, alphaBoolBlend {blendOption=Linear}, boneToModify {boneName=dyn_skate_c_l}}
   n20 = "Transform (Modify) Bone - Bone: dyn_skate_d_l"(ComponentPose=@n19, Alpha=@n16, Translation="0,0,0", Rotation=@n15, Scale="1,1,1") {rotationMode=BMM_Additive, rotationSpace=BCS_BoneSpace, alphaBoolBlend {blendOption=Linear}, boneToModify {boneName=dyn_skate_d_l}}
   n21 = Get Spinning_Rotator()
   n22 = Get PlayerMovement()
   n23 = "Transform (Modify) Bone - Bone: dyn_skate_a_r"(ComponentPose=@n20, Alpha=@n22, Translation="0,0,0", Rotation=@n21, Scale="1,1,1") {rotationMode=BMM_Additive, rotationSpace=BCS_BoneSpace, alphaBoolBlend {blendOption=Linear}, boneToModify {boneName=dyn_skate_a_r}}
   n24 = "Transform (Modify) Bone - Bone: dyn_skate_b_r"(ComponentPose=@n23, Alpha=@n22, Translation="0,0,0", Rotation=@n21, Scale="1,1,1") {rotationMode=BMM_Additive, rotationSpace=BCS_BoneSpace, alphaBoolBlend {blendOption=Linear}, boneToModify {boneName=dyn_skate_b_r}}
   n25 = "Transform (Modify) Bone - Bone: dyn_skate_c_r"(ComponentPose=@n24, Alpha=@n22, Translation="0,0,0", Rotation=@n21, Scale="1,1,1") {rotationMode=BMM_Additive, rotationSpace=BCS_BoneSpace, alphaBoolBlend {blendOption=Linear}, boneToModify {boneName=dyn_skate_c_r}}
   n26 = "Transform (Modify) Bone - Bone: dyn_skate_d_r"(ComponentPose=@n25, Alpha=@n22, Translation="0,0,0", Rotation=@n21, Scale="1,1,1") {rotationMode=BMM_Additive, rotationSpace=BCS_BoneSpace, alphaBoolBlend {blendOption=Linear}, boneToModify {boneName=dyn_skate_d_r}}
   Output Pose(Result=Component To Local(ComponentPose=@n26)) {layerGroup=None}
  ]}
 EventGraph {nodes [
   n1 = Event BlueprintUpdateAnimation()
   > n2 = Sequence!(execute) -> (then_0, then_1, then_2:Exec)
   n3 = float / float(A=GetRealTimeSeconds(), B=-720) {is=title K2Node_CallFunction}
   > Set Spinning_Rotator(Spinning_Rotator:Rotator=MakeRotator(Pitch=float * float(A=@n3, B=900000))) -> (Output_Get:Rotator)
   n4 = Get PawnSpeed2D()
   > Set PlayerMovement(PlayerMovement:double=SelectFloat(A=1, B=0, bPickA="float > float"(A=@n4, B=1))) -> (Output_Get:double)
   Set LegRotation(
    execute=@n2.then_2
    LegRotation:double=GetAlphaFromReferenceTransformDelta(SkeletalMeshComponent=Get Body_Mesh(), BoneName=calf_r, BaseName=thigh_r, InputRange=(X=12,Y=114), TwistAxis="1,0,0")
   ) -> (Output_Get:double)
   Is Valid(exec=@n2.then_1, InputObject=Get PlayerPawn() -> (PlayerPawn:Fort Player Pawn Athena Object Reference))
   > Set WPO Globals Function(
    Material=Get MatInstanceDynamic() -> (MatInstanceDynamic:Material Instance Dynamic Object Reference)
    Overlay Material=Get OverlayMID() -> (OverlayMID:Material Instance Dynamic Object Reference)
    Skeletal Mesh=Get SkelMesh()
    Local Noise Offset Multiplier=0.3
    "Local Noise Transform  Location"=Get Local Noise Transform Location() -> (Local Noise Transform Location:Vector)
    Local Noise Offset=Get Local Noise Offset()
   )
   > Set WPO Regional Function!(
    execute
    Material:Material Instance Dynamic Object Reference=Get MatInstanceDynamic()
    Overlay Material:Material Instance Dynamic Object Reference=Get OverlayMID()
    Skeletal Mesh:Skeletal Mesh Component Object Reference=Get SkelMesh()
    Velocity Transform:Name=hand_r
    Velocity Interp Speed:double=Get Velocity Interp Speed()
    Collision Transform:Name=lowerarm_r
    Collision Transform Axis:EAxis Enum
    Invert Collision Transform Axis:bool=false
    Delta Time X:double=@n1.DeltaTimeX
    Previous Velocity Transform Location:"Vector (by ref)"=Get R1 Velocity Transform Location()
    Previous Velocity Transform Velocity:"Vector (by ref)"=Get R1 Velocity Transform Velocity()
    Region:int=1
   ) -> (then)
   > Set WPO Regional Function(
    Material=Get MatInstanceDynamic()
    Overlay Material=Get OverlayMID()
    Skeletal Mesh=Get SkelMesh()
    Velocity Transform=hand_l
    Velocity Interp Speed=Get Velocity Interp Speed()
    Collision Transform=lowerarm_l
    Invert Collision Transform Axis=true
    Delta Time X=@n1.DeltaTimeX
    Previous Velocity Transform Location=Get R2 Velocity Bone Location() -> (R2 Velocity Bone Location:Vector)
    Previous Velocity Transform Velocity=Get R2 Velocity Transform Velocity()
    Region=2
   )
   > n5 = Branch(Condition=Get bIsSkyDiving())
   > n6 = Set WPO Pho Wind Intensity(WPO Pho Wind Intensity=1)
   SetScalarParameterValue(execute=[@n7.then, @n6.then], self=Get MatInstanceDynamic(), ParameterName=ClothSim_WindStrength, Value=Get WPO Pho Wind Intensity())
   n7 = Set WPO Pho Wind Intensity(execute=@n5.else, WPO Pho Wind Intensity=0)
   Event BlueprintInitializeAnimation()
   > Set SkelMesh(SkelMesh=GetOwnerSkeletalMeshComponent())
   n8 = TryGetPawnOwner()
   > Is Valid(InputObject=@n8)
   > n9 = Cast To FortPlayerPawnAthena(Object=@n8)
   > n10 = Set PlayerPawn(PlayerPawn:Fort Player Pawn Athena Object Reference=@n9) -> (Output_Get:Fort Player Pawn Athena Object Reference)
   > Set Body_Mesh(Body_Mesh=GetSkeletalMeshForPartType(self=@n10))
   Event OnCharacterCustomizationLoaded()
   > n11 = Set SkelMesh(SkelMesh=GetPartSkeletalMeshComponent())
   > n12 = Get Dynamic Material Instance(SkeletalMeshComponent=@n11)
   > Set MatInstanceDynamic(MatInstanceDynamic:Material Instance Dynamic Object Reference=@n12) -> (Output_Get:Material Instance Dynamic Object Reference)
   > n13 = GetOrCreateDynamicOverlayMaterial(MeshComponent=@n11, MaterialName=None)
   > Set OverlayMID(OverlayMID:Material Instance Dynamic Object Reference=@n13) -> (Output_Get:Material Instance Dynamic Object Reference)
  ]}
]
variables [
 Body_Mesh {type=SkeletalMeshComponent}
 LegRotation {type=number}
 Local Noise Offset {type=Vector}
 Local Noise Transform Location {type=Vector}
 MatInstanceDynamic {type=MaterialInstanceDynamic}
 OverlayMID {type=MaterialInstanceDynamic}
 PlayerMovement {type=number}
 PlayerPawn {type=FortPlayerPawnAthena}
 R1 Velocity Transform Location {type=Vector}
 R1 Velocity Transform Velocity {type=Vector}
 R2 Velocity Bone Location {type=Vector}
 R2 Velocity Transform Velocity {type=Vector}
 SkelMesh {type=SkeletalMeshComponent}
 Spinning_Rotator {type=Rotator}
 Velocity Interp Speed {type=number, default=30}
 WPO Pho Wind Intensity {type=number}
]
