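# Physics asset for the F_MED_Nox face accessory mesh, setting up dynamic hair on the female medium character Nox.
# Nine hair chains (a–i) run from small base spheres through one to three capsules, joined by constraints with tight swing and twist limits.
# An angularly free constraint cross-links the tips of chains c and d.
# Unconstrained capsules cover neck, upper spine, both clavicles and only the right upper arm.
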PhysicsAsset /BRCosmetics/Characters/Player/Female/Medium/Bodies/F_MED_Nox/Meshes/Parts/F_MED_Nox_FaceAcc_Physics
bodies {
 clavicle_l [{
   shapeType=Capsule
   length=13.4534
   radius=6.7778
   shapeName=clavicle_r_capsule
   center=<7.1997 -3.7414 -3.2212>
   rotation=<rot 86.9139 -0.0012 179.9988>
  }]
 clavicle_r [{
   shapeType=Capsule
   length=13.4534
   radius=6.7778
   shapeName=clavicle_r_capsule
   center=<-6.7362 2.7941 4.3629>
   rotation=<rot 63.7032 -0.0001 -0.0001>
  }]
 dyn_hair_a_01 [{shapeType=Capsule, length=3.8207, radius=3.2481, shapeName=dyn_hair_a_01_capsule, center=<2.9914 0 0>, rotation=<rot 0 0 79.6081>}]
 dyn_hair_a_02 [{shapeType=Capsule, length=4.7665, radius=3.8313, shapeName=dyn_hair_a_02_capsule, center=<3.2455 0 0>, rotation=<rot 0 0 79.6081>}]
 dyn_hair_a_base [{shapeType=Sphere, radius=0.5, shapeName=dyn_hair_a_base_sphere}]
 dyn_hair_b_01 [{shapeType=Capsule, length=3.6905, radius=2.5975, shapeName=dyn_hair_b_01_capsule, center=<3.1752 0.0708 0>, rotation=<rot 0 9.1042 90>}]
 dyn_hair_b_02 [{shapeType=Capsule, length=5.2817, radius=3.6499, shapeName=dyn_hair_b_02_capsule, center=<2.9667 0 0>, rotation=<rot 0 0 90>}]
 dyn_hair_b_base [{shapeType=Sphere, radius=0.5, shapeName=dyn_hair_b_base_sphere}]
 dyn_hair_c_01 [{shapeType=Capsule, length=4.6359, radius=3.5422, shapeName=dyn_hair_c_01_capsule, center=<3.6857 0.1364 0>, rotation=<rot 0 14.1031 90>}]
 dyn_hair_c_02 [{shapeType=Capsule, length=4.9351, radius=3.901, shapeName=dyn_hair_c_02_capsule, center=<3.5222 0 0>, rotation=<rot 0 0 90>}]
 dyn_hair_c_base [{shapeType=Sphere, radius=0.5, shapeName=dyn_hair_c_base_sphere}]
 dyn_hair_d_01 [{shapeType=Capsule, length=5.6251, radius=3.9237, shapeName=dyn_hair_d_01_capsule, center=<3.2036 0 0>, rotation=<rot 0 0 90>}]
 dyn_hair_d_02 [{shapeType=Capsule, length=7.3345, radius=2.9652, shapeName=dyn_hair_d_02_capsule, center=<2.6111 0 0>, rotation=<rot 0 0 90>}]
 dyn_hair_d_03 [{shapeType=Capsule, length=7.5495, radius=4.0512, shapeName=dyn_hair_d_03_capsule, center=<2.6897 0 0>, rotation=<rot 0 0 90>}]
 dyn_hair_d_base [{shapeType=Sphere, radius=0.5, shapeName=dyn_hair_d_base_sphere}]
 dyn_hair_e_01 [{shapeType=Capsule, length=5.4231, radius=3.5206, shapeName=dyn_hair_e_01_capsule, center=<3.8918 0 0>, rotation=<rot 0 -12.6134 90>}]
 dyn_hair_e_02 [{shapeType=Capsule, length=5.5936, radius=4.7011, shapeName=dyn_hair_e_02_capsule, center=<2.2282 0.0166 2.3302>, rotation=<rot 0 -5.5562 90>}]
 dyn_hair_e_base [{shapeType=Sphere, radius=0.5, shapeName=dyn_hair_e_base_sphere}]
 dyn_hair_f_01 [{shapeType=Capsule, length=3.4728, radius=2.9132, shapeName=dyn_hair_f_01_capsule, center=<2.3518 0 0>, rotation=<rot 0 0 90>}]
 dyn_hair_f_02 [{shapeType=Capsule, length=4.3822, radius=3.8514, shapeName=dyn_hair_f_02_capsule, center=<2.4812 0 0>, rotation=<rot 0 0 90>}]
 dyn_hair_f_base [{shapeType=Sphere, radius=0.5, shapeName=dyn_hair_f_base_sphere}]
 dyn_hair_g_01 [{shapeType=Capsule, length=4.2609, radius=3.6072, shapeName=dyn_hair_g_01_capsule, center=<3.5076 0 0>, rotation=<rot 0 -0.0001 126.1783>}]
 dyn_hair_g_02 [{shapeType=Capsule, length=5.8143, radius=3.1393, shapeName=dyn_hair_g_02_capsule, center=<1.8948 0 0>, rotation=<rot 0 -0.0001 126.1783>}]
 dyn_hair_g_base [{shapeType=Sphere, radius=0.5, shapeName=dyn_hair_g_base_sphere}]
 dyn_hair_h_01 [{shapeType=Capsule, length=3.8696, radius=3.2245, shapeName=dyn_hair_h_01_capsule, center=<0.6824 0 0>, rotation=<rot 90 0 0>}]
 dyn_hair_h_base [{shapeType=Sphere, radius=0.5, shapeName=dyn_hair_h_base_sphere}]
 dyn_hair_i_01 [{shapeType=Capsule, length=3.8696, radius=2.0847, shapeName=dyn_hair_i_01_capsule, center=<1.3918 0 1.6894>, rotation=<rot 90 0 0>}]
 dyn_hair_i_base [{shapeType=Sphere, radius=0.5, shapeName=dyn_hair_i_base_sphere}]
 hair_offset [{shapeType=Sphere, radius=0.5, shapeName=hair_offset_sphere}]
 neck_01 [{shapeType=Capsule, length=9.8383, radius=7.3188, shapeName=neck_01_capsule, center=<0 -0.7703 0>, rotation=<rot 90 0 0>}]
 spine_05 [{shapeType=Capsule, length=16.875, radius=13.125, shapeName=spine_05_capsule, center=<-5.0076 1.534 0>, rotation=<rot 90 180 176.5077>}]
 upperarm_r [{
   shapeType=Capsule
   length=17.6562
   radius=6.7812
   shapeName=upperarm_r_capsule
   center=<-13.9597 2.2824 6.995>
   rotation=<rot 82.2705 -0.0216 -0.0248>
  }]
}
constraints [
 {bone1Name=dyn_hair_g_02, bone2Name=dyn_hair_g_01, swing1LimitDegrees=20, swing2LimitDegrees=20, twistLimitDegrees=3}
 {bone1Name=dyn_hair_g_01, bone2Name=dyn_hair_g_base, swing1LimitDegrees=20, swing2LimitDegrees=20, twistLimitDegrees=3}
 {bone1Name=dyn_hair_f_01, bone2Name=dyn_hair_f_base, swing1LimitDegrees=10, swing2LimitDegrees=10, twistLimitDegrees=3}
 {bone1Name=dyn_hair_f_02, bone2Name=dyn_hair_f_01, swing1LimitDegrees=10, swing2LimitDegrees=10, twistLimitDegrees=3}
 {bone1Name=dyn_hair_e_02, bone2Name=dyn_hair_e_01, swing1LimitDegrees=15, swing2LimitDegrees=10, twistLimitDegrees=3}
 {bone1Name=dyn_hair_e_01, bone2Name=dyn_hair_e_base, swing1LimitDegrees=10, swing2LimitDegrees=10, twistLimitDegrees=3}
 {bone1Name=dyn_hair_d_03, bone2Name=dyn_hair_d_02, swing1LimitDegrees=15, swing2LimitDegrees=15, twistLimitDegrees=3}
 {bone1Name=dyn_hair_d_02, bone2Name=dyn_hair_d_01, swing1LimitDegrees=10, swing2LimitDegrees=15, twistLimitDegrees=3}
 {bone1Name=dyn_hair_d_01, bone2Name=dyn_hair_d_base, swing1LimitDegrees=10, swing2LimitDegrees=15, twistLimitDegrees=3}
 {bone1Name=dyn_hair_c_02, bone2Name=dyn_hair_c_01, swing1LimitDegrees=25, swing2LimitDegrees=30, twistLimitDegrees=5}
 {bone1Name=dyn_hair_c_01, bone2Name=dyn_hair_c_base, swing1LimitDegrees=20, swing2LimitDegrees=20, twistLimitDegrees=5}
 {bone1Name=dyn_hair_b_02, bone2Name=dyn_hair_b_01, swing1LimitDegrees=10, swing2LimitDegrees=20, twistLimitDegrees=3}
 {bone1Name=dyn_hair_b_01, bone2Name=dyn_hair_b_base, swing1LimitDegrees=10, swing2LimitDegrees=10, twistLimitDegrees=3}
 {bone1Name=dyn_hair_a_02, bone2Name=dyn_hair_a_01, swing1LimitDegrees=10, swing2LimitDegrees=15, twistLimitDegrees=3}
 {bone1Name=dyn_hair_a_01, bone2Name=dyn_hair_a_base, swing1LimitDegrees=5, swing2LimitDegrees=5, twistLimitDegrees=3}
 {bone1Name=dyn_hair_i_01, bone2Name=dyn_hair_i_base, swing1LimitDegrees=5, swing2LimitDegrees=10, twistLimitDegrees=1}
 {bone1Name=dyn_hair_h_01, bone2Name=dyn_hair_h_base, swing1LimitDegrees=5, swing2LimitDegrees=5, twistLimitDegrees=1}
 {
  bone1Name=dyn_hair_c_02
  bone2Name=dyn_hair_d_03
  swing1LimitDegrees=45
  swing1Motion=Free
  swing2LimitDegrees=45
  swing2Motion=Free
  twistLimitDegrees=45
  twistMotion=Free
 }
]
