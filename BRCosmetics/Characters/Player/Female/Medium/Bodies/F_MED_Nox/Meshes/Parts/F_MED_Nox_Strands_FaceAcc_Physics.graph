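# Physics asset for the strands face-accessory mesh part of the female medium Nox player cosmetic body.
# It defines eleven short dynamic strand chains (a–k), each with a small sphere at the base bone and a capsule along the strand bone.
# Constraints tie each strand to its base with narrow per-strand swing limits and very tight twist, restricting motion to slight sway.
# Strand f is the only two-segment chain, with its second capsule constrained to the first.
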PhysicsAsset /BRCosmetics/Characters/Player/Female/Medium/Bodies/F_MED_Nox/Meshes/Parts/F_MED_Nox_Strands_FaceAcc_Physics
bodies {
 dyn_strand_a_01 [{shapeType=Capsule, length=5.5031, radius=1, shapeName=dyn_strand_a_01_capsule, center=<3.4869 0 0>, rotation=<rot 90 0 0>}]
 dyn_strand_a_base [{shapeType=Sphere, radius=0.5, shapeName=dyn_strand_a_base_sphere}]
 dyn_strand_b_01 [{shapeType=Capsule, length=4.7413, radius=1.468, shapeName=dyn_strand_b_01_capsule, center=<3.9112 0 0>, rotation=<rot 90 0 0>}]
 dyn_strand_b_base [{shapeType=Sphere, radius=0.5, shapeName=dyn_strand_b_base_sphere}]
 dyn_strand_c_01 [{shapeType=Capsule, length=4.1347, radius=1.6829, shapeName=dyn_strand_c_01_capsule, center=<2.5077 0 0>, rotation=<rot 90 0 0>}]
 dyn_strand_c_base [{shapeType=Sphere, radius=0.5, shapeName=dyn_strand_c_base_sphere}]
 dyn_strand_d_01 [{shapeType=Capsule, length=5.2108, radius=0.9721, shapeName=dyn_strand_d_01_capsule, center=<2.5334 0 0>, rotation=<rot 90 0 0>}]
 dyn_strand_d_base [{shapeType=Sphere, radius=0.5, shapeName=dyn_strand_d_base_sphere}]
 dyn_strand_e_01 [{shapeType=Capsule, length=5.8479, radius=0.9187, shapeName=dyn_strand_e_01_capsule, center=<3.4724 0 0>, rotation=<rot 90 0 0>}]
 dyn_strand_e_base [{shapeType=Sphere, radius=0.5, shapeName=dyn_strand_e_base_sphere}]
 dyn_strand_f_01 [{shapeType=Capsule, length=4.9888, radius=1.4864, shapeName=dyn_strand_f_01_capsule, center=<3.3115 0 0>, rotation=<rot 90 0 0>}]
 dyn_strand_f_02 [{shapeType=Capsule, length=3.4805, radius=1.6975, shapeName=dyn_strand_f_02_capsule, center=<2.2388 0 0>, rotation=<rot 90 0 0>}]
 dyn_strand_f_base [{shapeType=Sphere, radius=0.5, shapeName=dyn_strand_f_base_sphere}]
 dyn_strand_g_01 [{shapeType=Capsule, length=3.1763, radius=0.7973, shapeName=dyn_strand_g_01_capsule, center=<1.8887 0 0>, rotation=<rot 90 0 0>}]
 dyn_strand_g_base [{shapeType=Sphere, radius=0.5, shapeName=dyn_strand_g_base_sphere}]
 dyn_strand_h_01 [{shapeType=Capsule, length=3.485, radius=0.883, shapeName=dyn_strand_h_01_capsule, center=<2.164 0 0>, rotation=<rot 90 0 0>}]
 dyn_strand_h_base [{shapeType=Sphere, radius=0.5, shapeName=dyn_strand_h_base_sphere}]
 dyn_strand_i_01 [{shapeType=Capsule, length=3.211, radius=1.4396, shapeName=dyn_strand_i_01_capsule, center=<1.6492 0 0>, rotation=<rot 90 0 0>}]
 dyn_strand_i_base [{shapeType=Sphere, radius=0.5, shapeName=dyn_strand_i_base_sphere}]
 dyn_strand_j_01 [{shapeType=Capsule, length=2.7975, radius=1.3807, shapeName=dyn_strand_j_01_capsule, center=<2.2337 0 0>, rotation=<rot 90 0 0>}]
 dyn_strand_j_base [{shapeType=Sphere, radius=0.5, shapeName=dyn_strand_j_base_sphere}]
 dyn_strand_k_01 [{shapeType=Capsule, length=3.4544, radius=0.9207, shapeName=dyn_strand_k_01_capsule, center=<2.2309 0 0>, rotation=<rot 90 0 0>}]
 dyn_strand_k_base [{shapeType=Sphere, radius=0.5, shapeName=dyn_strand_k_base_sphere}]
}
constraints [
 {bone1Name=dyn_strand_a_01, bone2Name=dyn_strand_a_base, swing1LimitDegrees=10, swing2LimitDegrees=10, twistLimitDegrees=2}
 {bone1Name=dyn_strand_f_01, bone2Name=dyn_strand_f_base, swing1LimitDegrees=3, swing2LimitDegrees=10, twistLimitDegrees=2}
 {bone1Name=dyn_strand_e_01, bone2Name=dyn_strand_e_base, swing1LimitDegrees=8, swing2LimitDegrees=5, twistLimitDegrees=2}
 {bone1Name=dyn_strand_d_01, bone2Name=dyn_strand_d_base, swing1LimitDegrees=10, swing2LimitDegrees=10, twistLimitDegrees=2}
 {bone1Name=dyn_strand_c_01, bone2Name=dyn_strand_c_base, swing1LimitDegrees=3, swing2LimitDegrees=3, twistLimitDegrees=2}
 {bone1Name=dyn_strand_b_01, bone2Name=dyn_strand_b_base, swing1LimitDegrees=5, swing2LimitDegrees=5, twistLimitDegrees=2}
 {bone1Name=dyn_strand_g_01, bone2Name=dyn_strand_g_base, swing1LimitDegrees=10, swing2LimitDegrees=8, twistLimitDegrees=2}
 {bone1Name=dyn_strand_h_01, bone2Name=dyn_strand_h_base, swing1LimitDegrees=8, swing2LimitDegrees=5, twistLimitDegrees=2}
 {bone1Name=dyn_strand_i_01, bone2Name=dyn_strand_i_base, swing1LimitDegrees=5, swing2LimitDegrees=5, twistLimitDegrees=2}
 {bone1Name=dyn_strand_j_01, bone2Name=dyn_strand_j_base, swing1LimitDegrees=8, swing2LimitDegrees=13, twistLimitDegrees=2}
 {bone1Name=dyn_strand_f_02, bone2Name=dyn_strand_f_01, swing1LimitDegrees=5, swing2LimitDegrees=5, twistLimitDegrees=2}
 {bone1Name=dyn_strand_k_01, bone2Name=dyn_strand_k_base, swing1LimitDegrees=3, swing2LimitDegrees=5, twistLimitDegrees=2}
]
