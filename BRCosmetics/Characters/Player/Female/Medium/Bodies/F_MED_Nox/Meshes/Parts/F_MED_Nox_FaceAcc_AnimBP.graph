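# Animation Blueprint for the F_MED_Nox character's face accessory, posing and simulating its dynamic hair bones.
# It copies the attached parent's pose, then layers mostly additive hair-bone offsets weighted by skydiving, crouched aiming, aiming and non-aiming alphas.
# Head twist against spine_05 drives the aim-related alphas, split by targeting state.
# Two head-space RigidBody nodes then simulate strands and accessory, with external force from a deprecated gravity-override function supplying skydive wind.
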AnimBlueprint /BRCosmetics/Characters/Player/Female/Medium/Bodies/F_MED_Nox/Meshes/Parts/F_MED_Nox_FaceAcc_AnimBP
target_skeleton=&/F_MED_Nox_FaceAcc_Skeleton
graphs [
 AnimGraph {nodes [
   n1 = "Transform (Modify) Bone - Bone: dyn_hair_b_base"(Alpha=1, Translation="-10,0,0", Rotation="0,0,0", Scale="1,1,1") {
    rotationMode=BMM_Additive
    rotationSpace=BCS_BoneSpace
    translationMode=BMM_Additive
    translationSpace=BCS_BoneSpace
    alphaBoolBlend {blendOption=Linear}
    alphaScaleBiasClamp {bInterpResult=true, interpSpeedDecreasing=4, interpSpeedIncreasing=6}
    boneToModify {boneName=dyn_hair_b_base}
    translation {x=-10}
   }
   n2 = "Transform (Modify) Bone - Bone: dyn_hair_c_base"(ComponentPose=@n1, Alpha=1, Translation="-10,-5,-1", Rotation="0,-60,0", Scale="1,1,1") {
    rotationMode=BMM_Additive
    rotationSpace=BCS_BoneSpace
    translationMode=BMM_Additive
    translationSpace=BCS_BoneSpace
    alphaBoolBlend {blendOption=Linear}
    alphaScaleBiasClamp {bInterpResult=true, interpSpeedDecreasing=4, interpSpeedIncreasing=6}
    boneToModify {boneName=dyn_hair_c_base}
    rotation {yaw=-60}
    translation=<-10 -5 -1>
   }
   "Transform (Modify) Bone - Bone: dyn_hair_e_base"(ComponentPose=@n2, Alpha=1, Translation="0,0,0", Rotation="-40,0,0", Scale="1,1,1") {
    rotationMode=BMM_Additive
    rotationSpace=BCS_BoneSpace
    translationMode=BMM_Additive
    translationSpace=BCS_BoneSpace
    alphaBoolBlend {blendOption=Linear}
    alphaScaleBiasClamp {bInterpResult=true, interpSpeedDecreasing=4, interpSpeedIncreasing=6}
    boneToModify {boneName=dyn_hair_e_base}
    rotation {pitch=-40}
   }
   n3 = Get Gravity_Override() -> (Gravity_Override:Vector)
   n4 = Get Is_Not_ADS_Rotate_Head() -> (Is_Not_ADS_Rotate_Head:double)
   n5 = Get Skydive_Alpha() -> (Skydive_Alpha:double)
   n6 = "Transform (Modify) Bone - Bone: dyn_hair_d_base"(
    ComponentPose=Local To Component(LocalPose=Copy Pose From Mesh() {bUseAttachedParent=true})
    Alpha=@n5
    Translation="0,0,-4"
    Rotation="-40,0,0"
    Scale="1.25,1,1.25"
   ) {
    rotationMode=BMM_Additive
    rotationSpace=BCS_BoneSpace
    scaleMode=BMM_Additive
    scaleSpace=BCS_BoneSpace
    translationMode=BMM_Additive
    translationSpace=BCS_BoneSpace
    alphaBoolBlend {blendOption=Linear}
    alphaScaleBiasClamp {bInterpResult=true}
    boneToModify {boneName=dyn_hair_d_base}
    rotation {pitch=-40}
    scale {x=1.25, z=1.25}
    translation {z=-4}
   }
   n7 = "Transform (Modify) Bone - Bone: dyn_hair_e_base"(ComponentPose=@n6, Alpha=@n5, Translation="0,0,-3", Rotation="-70,60,0", Scale="1.25,1,1.25") {
    rotationMode=BMM_Additive
    rotationSpace=BCS_BoneSpace
    scaleMode=BMM_Additive
    scaleSpace=BCS_BoneSpace
    translationMode=BMM_Additive
    translationSpace=BCS_BoneSpace
    alphaBoolBlend {blendOption=Linear}
    alphaScaleBiasClamp {bInterpResult=true}
    boneToModify {boneName=dyn_hair_e_base}
    rotation {pitch=-70, yaw=60}
    scale {x=1.25, z=1.25}
    translation {z=-3}
   }
   n8 = "Transform (Modify) Bone - Bone: dyn_hair_c_base"(ComponentPose=@n7, Alpha=@n5, Translation="0,0,-2", Rotation="-60,0,60", Scale="1.25,1,1.25") {
    rotationMode=BMM_Additive
    rotationSpace=BCS_BoneSpace
    scaleMode=BMM_Additive
    scaleSpace=BCS_BoneSpace
    translationMode=BMM_Additive
    translationSpace=BCS_BoneSpace
    alphaBoolBlend {blendOption=Linear}
    alphaScaleBiasClamp {bInterpResult=true}
    boneToModify {boneName=dyn_hair_c_base}
    rotation {pitch=-60, roll=60}
    scale {x=1.25, z=1.25}
    translation {z=-2}
   }
   n9 = Get Crouch_ADS_Head_Rotate() -> (Crouch_ADS_Head_Rotate:double)
   n10 = "Transform (Modify) Bone - Bone: dyn_hair_b_base"(ComponentPose=@n8, Alpha=@n9, Translation="-10,0,0", Rotation="0,0,0", Scale="1,1,1") {
    rotationMode=BMM_Additive
    rotationSpace=BCS_BoneSpace
    translationMode=BMM_Additive
    translationSpace=BCS_BoneSpace
    alphaBoolBlend {blendOption=Linear}
    alphaScaleBiasClamp {bInterpResult=true, interpSpeedDecreasing=4, interpSpeedIncreasing=6}
    boneToModify {boneName=dyn_hair_b_base}
    translation {x=-10}
   }
   n11 = "Transform (Modify) Bone - Bone: dyn_hair_c_base"(ComponentPose=@n10, Alpha=@n9, Translation="-5,-2,-2", Rotation="0,-50,0", Scale="1,1,1") {
    rotationMode=BMM_Additive
    rotationSpace=BCS_BoneSpace
    translationMode=BMM_Additive
    translationSpace=BCS_BoneSpace
    alphaBoolBlend {blendOption=Linear}
    alphaScaleBiasClamp {bInterpResult=true, interpSpeedDecreasing=4, interpSpeedIncreasing=6}
    boneToModify {boneName=dyn_hair_c_base}
    rotation {yaw=-50}
    translation=<-5 -2 -2>
   }
   n12 = "Transform (Modify) Bone - Bone: dyn_hair_e_base"(ComponentPose=@n11, Alpha=@n9, Translation="0,10,-3", Rotation="-40,20,0", Scale="1.25,1.25,1.25") {
    rotationMode=BMM_Additive
    rotationSpace=BCS_BoneSpace
    scaleMode=BMM_Additive
    scaleSpace=BCS_BoneSpace
    translationMode=BMM_Additive
    translationSpace=BCS_BoneSpace
    alphaBoolBlend {blendOption=Linear}
    alphaScaleBiasClamp {bInterpResult=true, interpSpeedDecreasing=4, interpSpeedIncreasing=6}
    boneToModify {boneName=dyn_hair_e_base}
    rotation {pitch=-40, yaw=20}
    scale=<1.25 1.25 1.25>
    translation {y=10, z=-3}
   }
   n13 = "Transform (Modify) Bone - Bone: dyn_hair_e_02"(ComponentPose=@n12, Alpha=@n9, Translation="0,0,0", Rotation="0,0,0", Scale="1.5,1.5,2") {
    rotationMode=BMM_Additive
    rotationSpace=BCS_BoneSpace
    scaleMode=BMM_Additive
    scaleSpace=BCS_BoneSpace
    translationMode=BMM_Additive
    translationSpace=BCS_BoneSpace
    alphaBoolBlend {blendOption=Linear}
    alphaScaleBiasClamp {bInterpResult=true, interpSpeedDecreasing=4, interpSpeedIncreasing=6}
    boneToModify {boneName=dyn_hair_e_02}
    scale=<1.5 1.5 2>
   }
   n14 = "Transform (Modify) Bone - Bone: dyn_hair_d_base"(ComponentPose=@n13, Alpha=@n9, Translation="0,0,-3", Rotation="0,0,0", Scale="1.5,1.5,2") {
    rotationMode=BMM_Additive
    rotationSpace=BCS_BoneSpace
    scaleMode=BMM_Additive
    scaleSpace=BCS_BoneSpace
    translationMode=BMM_Additive
    translationSpace=BCS_BoneSpace
    alphaBoolBlend {blendOption=Linear}
    boneToModify {boneName=dyn_hair_d_base}
    scale=<1.5 1.5 2>
    translation {z=-3}
   }
   n15 = "Transform (Modify) Bone - Bone: dyn_hair_d_02"(ComponentPose=@n14, Alpha=@n9, Translation="0,0,0", Rotation="0,0,0", Scale="1.5,1.5,2") {
    rotationMode=BMM_Additive
    rotationSpace=BCS_BoneSpace
    scaleMode=BMM_Additive
    scaleSpace=BCS_BoneSpace
    translationMode=BMM_Additive
    translationSpace=BCS_BoneSpace
    alphaBoolBlend {blendOption=Linear}
    boneToModify {boneName=dyn_hair_d_02}
    scale=<1.5 1.5 2>
   }
   n16 = "Transform (Modify) Bone - Bone: dyn_hair_d_03"(ComponentPose=@n15, Alpha=@n9, Translation="0,0,0", Rotation="0,0,0", Scale="1.5,1.5,2") {
    rotationMode=BMM_Additive
    rotationSpace=BCS_BoneSpace
    scaleMode=BMM_Additive
    scaleSpace=BCS_BoneSpace
    translationMode=BMM_Additive
    translationSpace=BCS_BoneSpace
    alphaBoolBlend {blendOption=Linear}
    boneToModify {boneName=dyn_hair_d_03}
    scale=<1.5 1.5 2>
   }
   n17 = "Transform (Modify) Bone - Bone: dyn_hair_f_base"(ComponentPose=@n16, Alpha=@n9, Translation="0,8,-3", Rotation="-20,0,0", Scale="1,1,1") {
    rotationMode=BMM_Additive
    rotationSpace=BCS_BoneSpace
    scaleMode=BMM_Additive
    scaleSpace=BCS_BoneSpace
    translationMode=BMM_Additive
    translationSpace=BCS_BoneSpace
    alphaBoolBlend {blendOption=Linear}
    alphaScaleBiasClamp {bInterpResult=true, interpSpeedDecreasing=4, interpSpeedIncreasing=6}
    boneToModify {boneName=dyn_hair_f_base}
    rotation {pitch=-20}
    translation {y=8, z=-3}
   }
   n18 = Get ADS_Rotate_Head() -> (ADS_Rotate_Head:double)
   n19 = "Transform (Modify) Bone - Bone: dyn_hair_a_base"(ComponentPose=@n17, Alpha=@n18, Translation="-1.5,0,1", Rotation="-60,-18,12", Scale="1,1,1") {
    rotationMode=BMM_Additive
    rotationSpace=BCS_BoneSpace
    scaleMode=BMM_Replace
    scaleSpace=BCS_BoneSpace
    translationMode=BMM_Replace
    translationSpace=BCS_BoneSpace
    alphaBoolBlend {blendOption=Linear}
    boneToModify {boneName=dyn_hair_a_base}
    rotation=<rot -60 -18 12>
    translation {x=-1.5, z=1}
   }
   n20 = "Transform (Modify) Bone - Bone: dyn_hair_b_base"(ComponentPose=@n19, Alpha=@n18, Translation="-8,0,0", Rotation="-20,0,50", Scale="1,1,1") {
    rotationMode=BMM_Additive
    rotationSpace=BCS_BoneSpace
    translationMode=BMM_Additive
    translationSpace=BCS_BoneSpace
    alphaBoolBlend {blendOption=Linear}
    boneToModify {boneName=dyn_hair_b_base}
    rotation {pitch=-20, roll=50}
    translation {x=-8}
   }
   n21 = "Transform (Modify) Bone - Bone: dyn_hair_c_base"(ComponentPose=@n20, Alpha=@n18, Translation="-8,0,-6", Rotation="-70,0,60", Scale="1,1,1") {
    rotationMode=BMM_Additive
    rotationSpace=BCS_BoneSpace
    scaleMode=BMM_Additive
    scaleSpace=BCS_BoneSpace
    translationMode=BMM_Additive
    translationSpace=BCS_BoneSpace
    alphaBoolBlend {blendOption=Linear}
    alphaScaleBiasClamp {bInterpResult=true}
    boneToModify {boneName=dyn_hair_c_base}
    rotation {pitch=-70, roll=60}
    translation {x=-8, z=-6}
   }
   n22 = "Transform (Modify) Bone - Bone: dyn_hair_d_base"(ComponentPose=@n21, Alpha=@n18, Translation="-2,0,0", Rotation="0,0,0", Scale="1,1,1") {
    rotationMode=BMM_Additive
    rotationSpace=BCS_BoneSpace
    translationMode=BMM_Additive
    translationSpace=BCS_BoneSpace
    alphaBoolBlend {blendOption=Linear}
    boneToModify {boneName=dyn_hair_d_base}
    translation {x=-2}
   }
   n23 = "Transform (Modify) Bone - Bone: dyn_hair_b_base"(ComponentPose=@n22, Alpha=@n4, Translation="0,0,0", Rotation="-10,-95,50", Scale="1,1,1") {
    rotationMode=BMM_Additive
    rotationSpace=BCS_BoneSpace
    scaleMode=BMM_Replace
    scaleSpace=BCS_BoneSpace
    translationMode=BMM_Additive
    translationSpace=BCS_BoneSpace
    alphaBoolBlend {blendOption=Linear}
    boneToModify {boneName=dyn_hair_b_base}
    rotation=<rot -10 -95 50>
   }
   n24 = "Transform (Modify) Bone - Bone: dyn_hair_a_base"(ComponentPose=@n23, Alpha=@n4, Translation="0,0,0", Rotation="-25,-30,0", Scale="1,1,1") {
    rotationMode=BMM_Additive
    rotationSpace=BCS_BoneSpace
    scaleMode=BMM_Replace
    scaleSpace=BCS_BoneSpace
    translationMode=BMM_Additive
    translationSpace=BCS_BoneSpace
    alphaBoolBlend {blendOption=Linear}
    boneToModify {boneName=dyn_hair_a_base}
    rotation {pitch=-25, yaw=-30}
   }
   n25 = "Transform (Modify) Bone - Bone: dyn_hair_c_base"(ComponentPose=@n24, Alpha=@n4, Translation="0,0,0", Rotation="-18,0,24", Scale="1,1,1") {
    rotationMode=BMM_Additive
    rotationSpace=BCS_BoneSpace
    scaleMode=BMM_Replace
    scaleSpace=BCS_BoneSpace
    translationMode=BMM_Additive
    translationSpace=BCS_BoneSpace
    alphaBoolBlend {blendOption=Linear}
    boneToModify {boneName=dyn_hair_c_base}
    rotation {pitch=-18, roll=24}
   }
   n26 = RigidBody(ComponentPose=@n25, ExternalForce=@n3) {
    bClampLinearTranslationLimitToRefPose=true
    bOverrideWorldGravity=true
    overridePhysicsAsset=&/F_MED_Nox_Strands_FaceAcc_Physics
    simulationSpace=BaseBoneSpace
    alphaBoolBlend {blendOption=Linear}
    baseBoneRef {boneName=head}
    componentAppliedLinearAccClamp=<6000 6000 6000>
    componentLinearAccScale=<0.2 0.2 0.25>
    componentLinearVelScale=<0.2 0.2 0.25>
    overrideWorldGravity {z=-980}
    simSpaceSettings {
     maxAngularAcceleration=100
     maxAngularVelocity=3
     maxLinearAcceleration=400
     maxLinearVelocity=250
     velocityScaleZ=0
     worldAlpha=1
     externalLinearDragV=<0.1 0.1>
    }
   }
   n27 = RigidBody(ComponentPose=@n26, ExternalForce=@n3) {
    bClampLinearTranslationLimitToRefPose=true
    bOverrideWorldGravity=true
    overridePhysicsAsset=&/F_MED_Nox_FaceAcc_Physics
    simulationSpace=BaseBoneSpace
    alphaBoolBlend {blendOption=Linear}
    baseBoneRef {boneName=head}
    componentAppliedLinearAccClamp=<6000 6000 6000>
    componentLinearAccScale=<0.2 0.2 0.35>
    componentLinearVelScale=<0.2 0.2 0.35>
    overrideWorldGravity {z=-980}
    simSpaceSettings {
     maxAngularAcceleration=100
     maxAngularVelocity=3
     maxLinearAcceleration=400
     maxLinearVelocity=250
     velocityScaleZ=0
     worldAlpha=1
     externalLinearDragV=<0.1 0.1>
    }
   }
   Output Pose(Result=Component To Local(ComponentPose=@n27)) {layerGroup=None}
  ]}
 EventGraph {nodes [
   Event BlueprintInitializeAnimation()
   > n1 = Cast To PlayerPawn_Athena(Object=TryGetPawnOwner())
   > n2 = Set Pawn Owner(Pawn Owner:Player Pawn Athena Object Reference=@n1) -> (Output_Get:Player Pawn Athena Object Reference)
   > Set Body_Mesh(Body_Mesh=GetSkeletalMeshForPartType(self=@n2))
   Event BlueprintUpdateAnimation()
   > n3 = Sequence!(execute) -> (then_0, then_1, then_2:Exec, then_3:Exec)
   > n4 = Set HeadRotate_Alpha(HeadRotate_Alpha:double=GetAlphaFromReferenceTransformDelta(
     SkeletalMeshComponent=Get Body_Mesh()
     BoneName=head
     BaseName=spine_05
     DeltaComponentType=QuaternionTwist
     InputRange=(X=-6,Y=50)
     TwistAxis="0,0,1"
    )) -> (Output_Get:double)
   > n5 = Branch(Condition=Get bIsTargeting())
   > Set ADS_Rotate_Head(ADS_Rotate_Head:double=@n4) -> (Output_Get:double)
   > Set Is_Not_ADS_Rotate_Head(Is_Not_ADS_Rotate_Head:double=0) -> (Output_Get:double)
   Set ADS_Rotate_Head(execute=@n5.else, ADS_Rotate_Head=0)
   > Set Is_Not_ADS_Rotate_Head(Is_Not_ADS_Rotate_Head=@n4)
   Set Crouch_ADS_Head_Rotate(execute=@n3.then_3, Crouch_ADS_Head_Rotate:double=SelectFloat(A=Get ADS_Rotate_Head(), B=0, bPickA=Get bIsCrouching())) -> (Output_Get:double)
   n6 = Get bIsSkydiveFloating()
   n7 = Get bIsSkydiveDiveMode()
   Set Skydive_Alpha(execute=@n3.then_2, Skydive_Alpha:double=SelectFloat(A=1, B=0, bPickA=OR Boolean(A=@n6, B=@n7))) -> (Output_Get:double)
   n8 = Gravity Override State Function DEPRECATED(
    execute=@n3.then_1
    CharacterPartAnimInstance=Self-Reference()
    GameTimeInSeconds=GetGameTimeInSeconds()
    GravityOverrideParams=Get Gravity Override Params()
   )
   > Set Gravity_Override(Gravity_Override:Vector=@n8) -> (Output_Get:Vector)
   GetAlphaFromReferenceTransformDelta(BoneName=head, BaseName=spine_05, DeltaComponentType=QuaternionTwist, InputRange=(X=-23,Y=0), TwistAxis="0,0,1")
   AND Boolean(A=Get bIsTargeting(), B=false)
   Get HeadRotate_Alpha() -> (HeadRotate_Alpha:double)
  ]}
]
variables [
 ADS_Rotate_Head {type=number}
 Body_Mesh {type=SkeletalMeshComponent}
 Crouch_ADS_Head_Rotate {type=number}
 Gravity Override Params {
  type=GravityOverrideParamsStruct
  default {
   skydive_DivingDown {windAmplitude=<60 400 200>, windFrequency=<2 14 12>}
   skydive_FloatingDown {gravityOverride=<0 -800 700>, windAmplitude=<300 -600 500>, windFrequency=<2 8 10>}
  }
 }
 Gravity_Override {type=Vector}
 HeadRotate_Alpha {type=number}
 HeadRotate_Bool {type=boolean}
 Is_Not_ADS_Rotate_Head {type=number}
 Pawn Owner {type=PlayerPawn_Athena_C}
 Skydive_Alpha {type=number}
]
